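# Physics asset for the hood of the male medium BraveBuildSuper body cosmetic, set up for secondary hood motion.
# Capsules cover the pelvis, spine and head, chained by free angular constraints.
# Seven tiny master spheres (front, side and back) hang freely off spine_05, each anchoring one dyn_hood capsule.
# Those capsules have tight swing and twist limits, with the center-back piece larger and allowed a wider swing.
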PhysicsAsset /BRCosmetics/Characters/Player/Male/Medium/Bodies/M_MED_BraveBuildSuper/Meshes/M_MED_BraveBuildSuper_Hood_PA
bodies {
 dyn_hood_back_01 [{shapeType=Capsule, length=10, radius=8.5, shapeName=dyn_hood_back_01_capsule, center=<4.5 0 4.5>, rotation=<rot 0 0 90>}]
 dyn_hood_back_left_01 [{shapeType=Capsule, length=5, radius=5, shapeName=dyn_hood_back_left_01_capsule, center=<1.5 0 0>, rotation=<rot 0 0 70>}]
 dyn_hood_back_right_01 [{shapeType=Capsule, length=5, radius=5, shapeName=dyn_hood_back_right_01_capsule, center=<1.5 0 0>, rotation=<rot 0 0 -70>}]
 dyn_hood_front_left_01 [{shapeType=Capsule, length=5, radius=5, shapeName=dyn_hood_front_left_01_capsule, center=<-4 0 0>, rotation=<rot 0 -20 90>}]
 dyn_hood_front_right_01 [{shapeType=Capsule, length=5, radius=5, shapeName=dyn_hood_front_right_01_capsule, center=<-4 0 0>, rotation=<rot 0 20 90>}]
 dyn_hood_left_01 [{shapeType=Capsule, length=5, radius=5, shapeName=dyn_hood_left_01_capsule, rotation=<rot 0 0 90>}]
 dyn_hood_right_01 [{shapeType=Capsule, length=5, radius=5, shapeName=dyn_hood_right_01_capsule, rotation=<rot 0 0 90>}]
 head [{shapeType=Capsule, length=10, radius=10, shapeName=head_capsule, rotation=<rot 90 0 0>}]
 hood_back_left_mstr [{shapeType=Sphere, radius=0.505, shapeName=hood_back_left_mstr_sphere}]
 hood_back_mstr [{shapeType=Sphere, radius=0.505, shapeName=hood_back_mstr_sphere}]
 hood_back_right_mstr [{shapeType=Sphere, radius=0.505, shapeName=hood_back_right_mstr_sphere}]
 hood_front_left_mstr [{shapeType=Sphere, radius=0.505, shapeName=hood_front_left_mstr_sphere}]
 hood_front_right_mstr [{shapeType=Sphere, radius=0.505, shapeName=hood_front_right_mstr_sphere}]
 hood_left_mstr [{shapeType=Sphere, radius=0.505, shapeName=hood_left_mstr_sphere}]
 hood_right_mstr [{shapeType=Sphere, radius=0.505, shapeName=hood_right_mstr_sphere}]
 pelvis [
  {shapeType=Capsule, length=6, radius=14, center=<-9 -3.5 0.25>}
  {shapeType=Capsule, length=14, radius=10, shapeName=pelvis_capsule, center=<0 -4 0>}
 ]
 spine_01 [{shapeType=Capsule, length=3, radius=15, shapeName=spine_01_capsule, center=<0 -0.5 0>}]
 spine_02 [{shapeType=Capsule, length=4, radius=14, shapeName=spine_02_capsule, center=<0 -0.5 0>}]
 spine_03 [{shapeType=Capsule, length=10, radius=11.5, shapeName=spine_03_capsule}]
 spine_04 [{shapeType=Capsule, length=12, radius=12, shapeName=spine_04_capsule}]
 spine_05 [{shapeType=Capsule, length=14, radius=12, shapeName=spine_05_capsule}]
}
constraints [
 {
  bone1Name=spine_01
  bone2Name=pelvis
  swing1LimitDegrees=45
  swing1Motion=Free
  swing2LimitDegrees=45
  swing2Motion=Free
  twistLimitDegrees=45
  twistMotion=Free
 }
 {
  bone1Name=spine_02
  bone2Name=spine_01
  swing1LimitDegrees=45
  swing1Motion=Free
  swing2LimitDegrees=45
  swing2Motion=Free
  twistLimitDegrees=45
  twistMotion=Free
 }
 {
  bone1Name=spine_03
  bone2Name=spine_02
  swing1LimitDegrees=45
  swing1Motion=Free
  swing2LimitDegrees=45
  swing2Motion=Free
  twistLimitDegrees=45
  twistMotion=Free
 }
 {
  bone1Name=spine_04
  bone2Name=spine_03
  swing1LimitDegrees=45
  swing1Motion=Free
  swing2LimitDegrees=45
  swing2Motion=Free
  twistLimitDegrees=45
  twistMotion=Free
 }
 {
  bone1Name=spine_05
  bone2Name=spine_04
  swing1LimitDegrees=45
  swing1Motion=Free
  swing2LimitDegrees=45
  swing2Motion=Free
  twistLimitDegrees=45
  twistMotion=Free
 }
 {
  bone1Name=hood_front_left_mstr
  bone2Name=spine_05
  swing1LimitDegrees=45
  swing1Motion=Free
  swing2LimitDegrees=45
  swing2Motion=Free
  twistLimitDegrees=45
  twistMotion=Free
 }
 {
  bone1Name=hood_left_mstr
  bone2Name=spine_05
  swing1LimitDegrees=45
  swing1Motion=Free
  swing2LimitDegrees=45
  swing2Motion=Free
  twistLimitDegrees=45
  twistMotion=Free
 }
 {
  bone1Name=hood_front_right_mstr
  bone2Name=spine_05
  swing1LimitDegrees=45
  swing1Motion=Free
  swing2LimitDegrees=45
  swing2Motion=Free
  twistLimitDegrees=45
  twistMotion=Free
 }
 {
  bone1Name=hood_right_mstr
  bone2Name=spine_05
  swing1LimitDegrees=45
  swing1Motion=Free
  swing2LimitDegrees=45
  swing2Motion=Free
  twistLimitDegrees=45
  twistMotion=Free
 }
 {bone1Name=dyn_hood_front_left_01, bone2Name=hood_front_left_mstr, swing1LimitDegrees=10, swing2LimitDegrees=25, twistLimitDegrees=5}
 {bone1Name=dyn_hood_front_right_01, bone2Name=hood_front_right_mstr, swing1LimitDegrees=10, swing2LimitDegrees=25, twistLimitDegrees=5}
 {bone1Name=dyn_hood_left_01, bone2Name=hood_left_mstr, swing1LimitDegrees=10, swing2LimitDegrees=25, twistLimitDegrees=5}
 {bone1Name=dyn_hood_right_01, bone2Name=hood_right_mstr, swing1LimitDegrees=10, swing2LimitDegrees=25, twistLimitDegrees=5}
 {
  bone1Name=head
  bone2Name=spine_05
  swing1LimitDegrees=45
  swing1Motion=Free
  swing2LimitDegrees=45
  swing2Motion=Free
  twistLimitDegrees=45
  twistMotion=Free
 }
 {
  bone1Name=hood_back_left_mstr
  bone2Name=spine_05
  swing1LimitDegrees=45
  swing1Motion=Free
  swing2LimitDegrees=45
  swing2Motion=Free
  twistLimitDegrees=45
  twistMotion=Free
 }
 {bone1Name=dyn_hood_back_left_01, bone2Name=hood_back_left_mstr, swing1LimitDegrees=10, swing2LimitDegrees=25, twistLimitDegrees=5}
 {
  bone1Name=hood_back_mstr
  bone2Name=spine_05
  swing1LimitDegrees=45
  swing1Motion=Free
  swing2LimitDegrees=45
  swing2Motion=Free
  twistLimitDegrees=45
  twistMotion=Free
 }
 {bone1Name=dyn_hood_back_01, bone2Name=hood_back_mstr, swing1LimitDegrees=10, swing2LimitDegrees=35, twistLimitDegrees=5}
 {
  bone1Name=hood_back_right_mstr
  bone2Name=spine_05
  swing1LimitDegrees=45
  swing1Motion=Free
  swing2LimitDegrees=45
  swing2Motion=Free
  twistLimitDegrees=45
  twistMotion=Free
 }
 {bone1Name=dyn_hood_back_right_01, bone2Name=hood_back_right_mstr, swing1LimitDegrees=10, swing2LimitDegrees=25, twistLimitDegrees=5}
]
