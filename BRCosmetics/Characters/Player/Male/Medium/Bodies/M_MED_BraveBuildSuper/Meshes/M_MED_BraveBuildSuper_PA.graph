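# Physics asset for the BraveBuildSuper male medium player body cosmetic, providing secondary motion for its jacket skirt and belt.
# Nine two-bone jacket capsule chains and one short belt capsule hang from small master-bone spheres jointed to the pelvis.
# Chain joints use tight twist and modest swing limits.
# Pelvis, spine and thigh capsules, linked by free angular joints, supply body volumes for the dangling pieces.
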PhysicsAsset /BRCosmetics/Characters/Player/Male/Medium/Bodies/M_MED_BraveBuildSuper/Meshes/M_MED_BraveBuildSuper_PA
bodies {
 belt_mstr [{shapeType=Sphere, radius=1, shapeName=belt_mstr_sphere}]
 dyn_belt_01 [{shapeType=Capsule, length=4, radius=3, shapeName=dyn_belt_01_capsule, center=<4 0 2>, rotation=<rot 90 0 0>}]
 dyn_jacket_bk_01 [{shapeType=Capsule, length=10, radius=6, shapeName=dyn_jacket_bk_01_capsule, center=<6 0 3>, rotation=<rot 90 0 0>}]
 dyn_jacket_bk_02 [{shapeType=Capsule, length=10, radius=6, shapeName=dyn_jacket_bk_02_capsule, center=<4 0 3>, rotation=<rot 90 0 0>}]
 dyn_jacket_bk_l_01 [{shapeType=Capsule, length=10, radius=5.75, shapeName=dyn_jacket_bk_l_01_capsule, center=<6 0 4.5>, rotation=<rot 90 0 0>}]
 dyn_jacket_bk_l_02 [{shapeType=Capsule, length=10, radius=5.75, shapeName=dyn_jacket_bk_l_02_capsule, center=<4 0 4.5>, rotation=<rot 90 0 0>}]
 dyn_jacket_bk_r_01 [{shapeType=Capsule, length=10, radius=6, shapeName=dyn_jacket_bk_r_01_capsule, center=<6 -3 2.5>, rotation=<rot 90 0 0>}]
 dyn_jacket_bk_r_02 [{shapeType=Capsule, length=10, radius=6, shapeName=dyn_jacket_bk_r_02_capsule, center=<4 -3 2.5>, rotation=<rot 90 0 0>}]
 dyn_jacket_bkout_l_01 [{shapeType=Capsule, length=10, radius=6, shapeName=dyn_jacket_bkout_l_01_capsule, center=<6 0 6.5>, rotation=<rot 90 0 0>}]
 dyn_jacket_bkout_l_02 [{shapeType=Capsule, length=10, radius=6, shapeName=dyn_jacket_bkout_l_02_capsule, center=<4 0 6.5>, rotation=<rot 90 0 0>}]
 dyn_jacket_bkout_r_01 [{shapeType=Capsule, length=10, radius=6, shapeName=dyn_jacket_bkout_r_01_capsule, center=<6 0 5>, rotation=<rot 90 0 0>}]
 dyn_jacket_bkout_r_02 [{shapeType=Capsule, length=10, radius=6, shapeName=dyn_jacket_bkout_r_02_capsule, center=<4 0 5>, rotation=<rot 90 0 0>}]
 dyn_jacket_fr_l_01 [{shapeType=Capsule, length=10, radius=6, shapeName=dyn_jacket_fr_l_01_capsule, center=<6 -4 4.5>, rotation=<rot 90 0 0>}]
 dyn_jacket_fr_l_02 [{shapeType=Capsule, length=10, radius=6, shapeName=dyn_jacket_fr_l_02_capsule, center=<4 -4 4.5>, rotation=<rot 90 0 0>}]
 dyn_jacket_fr_r_01 [{shapeType=Capsule, length=10, radius=6, shapeName=dyn_jacket_fr_r_01_capsule, center=<6 1 5.5>, rotation=<rot 90 0 0>}]
 dyn_jacket_fr_r_02 [{shapeType=Capsule, length=10, radius=6, shapeName=dyn_jacket_fr_r_02_capsule, center=<4 1 5.5>, rotation=<rot 90 0 0>}]
 dyn_jacket_l_01 [{shapeType=Capsule, length=10, radius=6, shapeName=dyn_jacket_l_01_capsule, center=<6 0 6>, rotation=<rot 90 0 0>}]
 dyn_jacket_l_02 [{shapeType=Capsule, length=10, radius=6, shapeName=dyn_jacket_l_02_capsule, center=<4 0 6>, rotation=<rot 90 0 0>}]
 dyn_jacket_r_01 [{shapeType=Capsule, length=10, radius=6, shapeName=dyn_jacket_r_01_capsule, center=<6 0 5.5>, rotation=<rot 90 0 0>}]
 dyn_jacket_r_02 [{shapeType=Capsule, length=10, radius=6, shapeName=dyn_jacket_r_02_capsule, center=<4 0 5.5>, rotation=<rot 90 0 0>}]
 jacket_bk_l_mstr [{shapeType=Sphere, radius=0.505, shapeName=jacket_bk_l_mstr_sphere}]
 jacket_bk_mstr [{shapeType=Sphere, radius=0.505, shapeName=jacket_bk_mstr_sphere}]
 jacket_bk_r_mstr [{shapeType=Sphere, radius=0.505, shapeName=jacket_bk_r_mstr_sphere}]
 jacket_bkout_l_mstr [{shapeType=Sphere, radius=0.505, shapeName=jacket_bkout_l_mstr_sphere}]
 jacket_bkout_r_mstr [{shapeType=Sphere, radius=0.505, shapeName=jacket_bkout_r_mstr_sphere}]
 jacket_fr_l_mstr [{shapeType=Sphere, radius=0.505, shapeName=jacket_fr_l_mstr_sphere}]
 jacket_fr_r_mstr [{shapeType=Sphere, radius=0.505, shapeName=jacket_fr_r_mstr_sphere}]
 jacket_l_mstr [{shapeType=Sphere, radius=0.505, shapeName=jacket_l_mstr_sphere}]
 jacket_r_mstr [{shapeType=Sphere, radius=0.505, shapeName=jacket_r_mstr_sphere}]
 pelvis [
  {shapeType=Capsule, length=6, radius=14, center=<-9 -3.5 0.25>}
  {shapeType=Capsule, length=14, radius=10, shapeName=pelvis_capsule, center=<0 -4 0>}
 ]
 spine_01 [{shapeType=Capsule, length=3, radius=15, shapeName=spine_01_capsule, center=<0 -0.5 0>}]
 spine_02 [{shapeType=Capsule, length=4, radius=14, shapeName=spine_02_capsule, center=<0 -0.5 0>}]
 spine_03 [{shapeType=Capsule, length=10, radius=11.5, shapeName=spine_03_capsule}]
 spine_04 [{shapeType=Capsule, length=12, radius=12, shapeName=spine_04_capsule}]
 spine_05 [{shapeType=Capsule, length=14, radius=12.5, shapeName=spine_05_capsule}]
 thigh_l [{shapeType=Capsule, length=24.5, radius=10.5, shapeName=thigh_l_capsule, center=<-20 0 0>, rotation=<rot 90 0 0>}]
 thigh_r [{shapeType=Capsule, length=24.5, radius=11, shapeName=thigh_r_capsule, center=<20 0 0>, rotation=<rot 90 0 0>}]
}
constraints [
 {
  bone1Name=spine_01
  bone2Name=pelvis
  swing1LimitDegrees=45
  swing1Motion=Free
  swing2LimitDegrees=45
  swing2Motion=Free
  twistLimitDegrees=45
  twistMotion=Free
 }
 {
  bone1Name=spine_02
  bone2Name=spine_01
  swing1LimitDegrees=45
  swing1Motion=Free
  swing2LimitDegrees=45
  swing2Motion=Free
  twistLimitDegrees=45
  twistMotion=Free
 }
 {
  bone1Name=spine_03
  bone2Name=spine_02
  swing1LimitDegrees=45
  swing1Motion=Free
  swing2LimitDegrees=45
  swing2Motion=Free
  twistLimitDegrees=45
  twistMotion=Free
 }
 {
  bone1Name=spine_04
  bone2Name=spine_03
  swing1LimitDegrees=45
  swing1Motion=Free
  swing2LimitDegrees=45
  swing2Motion=Free
  twistLimitDegrees=45
  twistMotion=Free
 }
 {
  bone1Name=spine_05
  bone2Name=spine_04
  swing1LimitDegrees=45
  swing1Motion=Free
  swing2LimitDegrees=45
  swing2Motion=Free
  twistLimitDegrees=45
  twistMotion=Free
 }
 {bone1Name=dyn_jacket_bk_l_02, bone2Name=dyn_jacket_bk_l_01, swing1LimitDegrees=25, swing2LimitDegrees=35, twistLimitDegrees=5}
 {bone1Name=dyn_jacket_bk_02, bone2Name=dyn_jacket_bk_01, swing1LimitDegrees=25, swing2LimitDegrees=35, twistLimitDegrees=5}
 {bone1Name=dyn_jacket_bk_r_02, bone2Name=dyn_jacket_bk_r_01, swing1LimitDegrees=25, swing2LimitDegrees=35, twistLimitDegrees=5}
 {bone1Name=dyn_jacket_bkout_l_02, bone2Name=dyn_jacket_bkout_l_01, swing1LimitDegrees=15, swing2LimitDegrees=35, twistLimitDegrees=5}
 {bone1Name=dyn_jacket_bkout_r_02, bone2Name=dyn_jacket_bkout_r_01, swing1LimitDegrees=15, swing2LimitDegrees=35, twistLimitDegrees=5}
 {bone1Name=dyn_jacket_fr_l_02, bone2Name=dyn_jacket_fr_l_01, swing1LimitDegrees=10, swing2LimitDegrees=30, twistLimitDegrees=5}
 {bone1Name=dyn_jacket_fr_r_02, bone2Name=dyn_jacket_fr_r_01, swing1LimitDegrees=10, swing2LimitDegrees=30, twistLimitDegrees=5}
 {bone1Name=dyn_jacket_l_02, bone2Name=dyn_jacket_l_01, swing1LimitDegrees=10, swing2LimitDegrees=30, twistLimitDegrees=5}
 {bone1Name=dyn_jacket_r_02, bone2Name=dyn_jacket_r_01, swing1LimitDegrees=10, swing2LimitDegrees=30, twistLimitDegrees=5}
 {
  bone1Name=thigh_l
  bone2Name=pelvis
  swing1LimitDegrees=45
  swing1Motion=Free
  swing2LimitDegrees=45
  swing2Motion=Free
  twistLimitDegrees=45
  twistMotion=Free
 }
 {
  bone1Name=thigh_r
  bone2Name=pelvis
  swing1LimitDegrees=45
  swing1Motion=Free
  swing2LimitDegrees=45
  swing2Motion=Free
  twistLimitDegrees=45
  twistMotion=Free
 }
 {bone1Name=dyn_jacket_bk_l_01, bone2Name=jacket_bk_l_mstr, swing1LimitDegrees=15, swing2LimitDegrees=35, twistLimitDegrees=5}
 {
  bone1Name=jacket_bk_l_mstr
  bone2Name=pelvis
  swing1LimitDegrees=45
  swing1Motion=Free
  swing2LimitDegrees=45
  swing2Motion=Free
  twistLimitDegrees=45
  twistMotion=Free
 }
 {bone1Name=dyn_jacket_bk_01, bone2Name=jacket_bk_mstr, swing1LimitDegrees=15, swing2LimitDegrees=35, twistLimitDegrees=5}
 {
  bone1Name=jacket_bk_mstr
  bone2Name=pelvis
  swing1LimitDegrees=45
  swing1Motion=Free
  swing2LimitDegrees=45
  swing2Motion=Free
  twistLimitDegrees=45
  twistMotion=Free
 }
 {bone1Name=dyn_jacket_bk_r_01, bone2Name=jacket_bk_r_mstr, swing1LimitDegrees=15, swing2LimitDegrees=35, twistLimitDegrees=5}
 {
  bone1Name=jacket_bk_r_mstr
  bone2Name=pelvis
  swing1LimitDegrees=45
  swing1Motion=Free
  swing2LimitDegrees=45
  swing2Motion=Free
  twistLimitDegrees=45
  twistMotion=Free
 }
 {bone1Name=dyn_jacket_bkout_l_01, bone2Name=jacket_bkout_l_mstr, swing1LimitDegrees=15, swing2LimitDegrees=35, twistLimitDegrees=5}
 {
  bone1Name=jacket_bkout_l_mstr
  bone2Name=pelvis
  swing1LimitDegrees=45
  swing1Motion=Free
  swing2LimitDegrees=45
  swing2Motion=Free
  twistLimitDegrees=45
  twistMotion=Free
 }
 {bone1Name=dyn_jacket_bkout_r_01, bone2Name=jacket_bkout_r_mstr, swing1LimitDegrees=15, swing2LimitDegrees=35, twistLimitDegrees=5}
 {
  bone1Name=jacket_bkout_r_mstr
  bone2Name=pelvis
  swing1LimitDegrees=45
  swing1Motion=Free
  swing2LimitDegrees=45
  swing2Motion=Free
  twistLimitDegrees=45
  twistMotion=Free
 }
 {bone1Name=dyn_jacket_fr_l_01, bone2Name=jacket_fr_l_mstr, swing1LimitDegrees=10, swing2LimitDegrees=30, twistLimitDegrees=5}
 {
  bone1Name=jacket_fr_l_mstr
  bone2Name=pelvis
  swing1LimitDegrees=45
  swing1Motion=Free
  swing2LimitDegrees=45
  swing2Motion=Free
  twistLimitDegrees=45
  twistMotion=Free
 }
 {bone1Name=dyn_jacket_fr_r_01, bone2Name=jacket_fr_r_mstr, swing1LimitDegrees=10, swing2LimitDegrees=30, twistLimitDegrees=5}
 {
  bone1Name=jacket_fr_r_mstr
  bone2Name=pelvis
  swing1LimitDegrees=45
  swing1Motion=Free
  swing2LimitDegrees=45
  swing2Motion=Free
  twistLimitDegrees=45
  twistMotion=Free
 }
 {bone1Name=dyn_jacket_l_01, bone2Name=jacket_l_mstr, swing1LimitDegrees=15, swing2LimitDegrees=35, twistLimitDegrees=5}
 {
  bone1Name=jacket_l_mstr
  bone2Name=pelvis
  swing1LimitDegrees=45
  swing1Motion=Free
  swing2LimitDegrees=45
  swing2Motion=Free
  twistLimitDegrees=45
  twistMotion=Free
 }
 {bone1Name=dyn_jacket_r_01, bone2Name=jacket_r_mstr, swing1LimitDegrees=10, swing2LimitDegrees=30, twistLimitDegrees=5}
 {
  bone1Name=jacket_r_mstr
  bone2Name=pelvis
  swing1LimitDegrees=45
  swing1Motion=Free
  swing2LimitDegrees=45
  swing2Motion=Free
  twistLimitDegrees=45
  twistMotion=Free
 }
 {
  bone1Name=belt_mstr
  bone2Name=pelvis
  swing1LimitDegrees=45
  swing1Motion=Free
  swing2LimitDegrees=45
  swing2Motion=Free
  twistLimitDegrees=45
  twistMotion=Free
 }
 {bone1Name=dyn_belt_01, bone2Name=belt_mstr, swing1LimitDegrees=1, swing2LimitDegrees=15, twistLimitDegrees=1}
]
