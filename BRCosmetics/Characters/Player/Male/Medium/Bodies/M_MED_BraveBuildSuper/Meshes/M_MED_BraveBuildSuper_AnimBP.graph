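# Post-process style cosmetic AnimBP for the M_MED_BraveBuildSuper body, animating coat, hood, zippers, belt, emblem and sleeves.
# The parent's copied pose gets a deform Control Rig and AnimDynamics, then is cached for two constraint and coat-rig branches.
# bRigidBodyEnabled selects RigidBody simulation, or a cheaper Trail Controller fallback weighted by movement.
# The event graph computes movement and skydiving weights and drives material WPO and wind.
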AnimBlueprint /BRCosmetics/Characters/Player/Male/Medium/Bodies/M_MED_BraveBuildSuper/Meshes/M_MED_BraveBuildSuper_AnimBP
target_skeleton=&/M_MED_BraveBuildSuper_Skeleton
graphs [
 AnimGraph {nodes [
   n1 = "Constraint - emblem_mstr (PPPPPPPP)"!(
    ComponentPose=Local To Component(LocalPose=Use cached pose 'Anim Dynamics'())
    Alpha=1
    ConstraintWeights_0=0.0828
    ConstraintWeights_1:float=0.1194
    ConstraintWeights_2:float=0.0635
    ConstraintWeights_3:float=0.1983
    ConstraintWeights_4:float=0.1315
    ConstraintWeights_5:float=0.1428
    ConstraintWeights_6:float=0.0281
    ConstraintWeights_7:float=0.2337
   ) -> (Pose) {
    alphaBoolBlend {blendOption=Linear}
    boneToModify {boneName=emblem_mstr}
    constraintSetup [
     {targetBone {boneName=spine_03}}
     {targetBone {boneName=spine_04}}
     {targetBone {boneName=spine_05}}
     {targetBone {boneName=clavicle_l}}
     {targetBone {boneName=upperarm_twist_01_l}}
     {targetBone {boneName=deform_upperarm_fr_l}}
     {targetBone {boneName=upperarm_twist_02_l}}
     {targetBone {boneName=deform_pec_l}}
    ]
    constraintWeights [1,1,1,1,1,1,1,1]
   }
   n2 = "Constraint - dyn_emblem_02 (PPPPPPPP)"!(
    ComponentPose=@n1
    Alpha=1
    ConstraintWeights_0=0.063
    ConstraintWeights_1:float=0.1685
    ConstraintWeights_2:float=0.2843
    ConstraintWeights_3:float=0.1618
    ConstraintWeights_4:float=0.118
    ConstraintWeights_5:float=0.0614
    ConstraintWeights_6:float=0.0255
    ConstraintWeights_7:float=0.1177
   ) -> (Pose) {
    alphaBoolBlend {blendOption=Linear}
    boneToModify {boneName=dyn_emblem_02}
    constraintSetup [
     {targetBone {boneName=spine_01}}
     {targetBone {boneName=spine_02}}
     {targetBone {boneName=spine_03}}
     {targetBone {boneName=spine_04}}
     {targetBone {boneName=clavicle_l}}
     {targetBone {boneName=upperarm_twist_01_l}}
     {targetBone {boneName=upperarm_twist_02_l}}
     {targetBone {boneName=deform_pec_l}}
    ]
    constraintWeights [1,1,1,1,1,1,1,1]
   }
   n3 = Look At - Bone: dyn_emblem_01(ComponentPose=@n2, Alpha=1) {
    alphaBoolBlend {blendOption=Linear}
    boneToModify {boneName=dyn_emblem_01}
    lookAtTarget {boneReference {boneName=dyn_emblem_02}}
    lookAt_Axis {axis=<1 0>}
   }
   n4 = "Transform (Modify) Bone - Bone: neck_correction_mstr"(ComponentPose=@n3, Alpha=1, Translation="0,0,0.75", Rotation="0,0,0", Scale="1,1,1") {
    translationMode=BMM_Additive
    translationSpace=BCS_BoneSpace
    alphaBoolBlend {blendOption=Linear}
    boneToModify {boneName=neck_correction_mstr}
    translation {z=0.75}
   }
   n5 = Control Rig!(
    Source=Component To Local(ComponentPose=@n4)
    Alpha
    SkyDiving_CTRL:bool=Get bIsSkyDiving()
    Crouch_CTRL:bool=Get bIsCrouching()
    RBAN Disabled:bool=NOT Boolean(A=Get bRigidBodyEnabled())
    RBAN Enabled:bool=Get bRigidBodyEnabled()
    Jumping_Falling:bool=OR Boolean(A=Get bIsJumping(), B=Get bIsFalling())
    Accelerating:bool=Get bIsAccelerating2D()
    Sliding:bool=Get bIsSliding()
    Emoting:bool=Get bIsPlayingEmote()
    Aiming:bool=Get bIsTargeting()
   ) -> (Pose) {
    alphaBoolBlend {blendOption=Linear}
    controlRigAssetReference {blueprintRigClass=&/M_MED_BraveBuildSuper_Coat_ControlRig.M_MED_BraveBuildSuper_Coat_ControlRig_C}
   }
   n6 = "Constraint - neck_zipper_mstr (PPPPPPPP)"!(
    ComponentPose=Local To Component(LocalPose=@n5)
    Alpha=1
    ConstraintWeights_0=0.0444
    ConstraintWeights_1:float=0.2825
    ConstraintWeights_2:float=0.0126
    ConstraintWeights_3:float=0.3079
    ConstraintWeights_4:float=0.1663
    ConstraintWeights_5:float=0.0527
    ConstraintWeights_6:float=0.0094
    ConstraintWeights_7:float=0.1242
   ) -> (Pose) {
    alphaBoolBlend {blendOption=Linear}
    boneToModify {boneName=neck_zipper_mstr}
    constraintSetup [
     {targetBone {boneName=spine_04}}
     {targetBone {boneName=spine_05}}
     {targetBone {boneName=clavicle_r}}
     {targetBone {boneName=neck_01}}
     {targetBone {boneName=neck_02}}
     {targetBone {boneName=head}}
     {targetBone {boneName=FACIAL_C_Jawline}}
     {targetBone {boneName=neck_correction_dyn}}
    ]
    constraintWeights [1,1,1,1,1,1,1,1]
   }
   n7 = Anim Dynamics - Bone: neck_zipper_mstr(ComponentPose=@n6, Alpha=1, ExternalForce="0,0,0") {
    angularDampingOverride=0.9999
    bOverrideAngularDamping=true
    bUseGravityOverride=true
    bUsePlanarLimit=false
    simulationSpace=RootRelative
    alphaBoolBlend {blendOption=Linear}
    boundBone {boneName=neck_zipper_mstr}
    componentAppliedLinearAccClamp=<3000 3000 3000>
    componentLinearAccScale=<0.5 0.5 0.5>
    componentLinearVelScale=<0.25 0.25 0.25>
    gravityOverride {z=-980}
    physicsBodyDefinitions [{
      boundBone {boneName=neck_zipper_mstr}
      boxExtents=<2 10 10>
      constraintSetup {angularLimitsMax=<2 5 30>, angularLimitsMin=<-2 -30 -30>, angularTarget=<0 0 0>}
      localJointOffset=<3 0 0>
     }]
   }
   n8 = "Constraint - belt_mstr (PPPPPPPP)"!(
    ComponentPose=@n7
    Alpha=1
    ConstraintWeights_0=0.5165
    ConstraintWeights_1:float=0
    ConstraintWeights_2:float=0.0017
    ConstraintWeights_3:float=0
    ConstraintWeights_4:float=0.1722
    ConstraintWeights_5:float=0.2228
    ConstraintWeights_6:float=0.0557
    ConstraintWeights_7:float=0.031
   ) -> (Pose) {
    alphaBoolBlend {blendOption=Linear}
    boneToModify {boneName=belt_mstr}
    constraintSetup [
     {targetBone {boneName=pelvis}}
     {targetBone {boneName=spine_01}}
     {targetBone {boneName=spine_02}}
     {targetBone {boneName=spine_03}}
     {targetBone {boneName=thigh_twist_01_l}}
     {targetBone {boneName=deform_groin_l}}
     {targetBone {boneName=thigh_twist_01_r}}
     {targetBone {boneName=deform_groin_r}}
    ]
    constraintWeights [1,1,1,1,1,1,1,1]
   }
   n9 = "Constraint - jacket_fr_r_mstr (PPPPP)"!(
    ComponentPose=@n8
    Alpha=1
    ConstraintWeights_0=0.253
    ConstraintWeights_1:float=0.2257
    ConstraintWeights_2:float=0.3059
    ConstraintWeights_3:float=0.1965
    ConstraintWeights_4:float=0.0189
   ) -> (Pose) {
    alphaBoolBlend {blendOption=Linear}
    boneToModify {boneName=jacket_fr_r_mstr}
    constraintSetup [
     {targetBone {boneName=pelvis}}
     {targetBone {boneName=spine_01}}
     {targetBone {boneName=spine_02}}
     {targetBone {boneName=spine_03}}
     {targetBone {boneName=spine_04}}
    ]
    constraintWeights [1,1,1,1,1]
   }
   n10 = RigidBody!(ComponentPose=@n9, Alpha, OverrideWorldGravity:Vector="0,0,-980", ExternalForce) -> (Pose) {
    bClampLinearTranslationLimitToRefPose=true
    bOverrideWorldGravity=true
    overridePhysicsAsset=&/M_MED_BraveBuildSuper_PA
    simulationSpace=BaseBoneSpace
    alphaBoolBlend {blendOption=Linear}
    baseBoneRef {boneName=pelvis}
    componentAppliedLinearAccClamp=<5000 5000 5000>
    componentLinearAccScale=<0.4 0.45 0.45>
    componentLinearVelScale=<0.2 0.25 0.25>
    simSpaceSettings {
     maxAngularAcceleration=1000
     maxAngularVelocity=3
     maxLinearAcceleration=700
     maxLinearVelocity=500
     velocityScaleZ=0
     worldAlpha=1
     externalLinearDragV=<0.5 0.5>
    }
   }
   n11 = RigidBody(ComponentPose=@n10, OverrideWorldGravity="0,0,-980") {
    bClampLinearTranslationLimitToRefPose=true
    bOverrideWorldGravity=true
    overridePhysicsAsset=&/M_MED_BraveBuildSuper_Hood_PA
    simulationSpace=BaseBoneSpace
    alphaBoolBlend {blendOption=Linear}
    baseBoneRef {boneName=spine_03}
    componentLinearAccScale=<0.3 0.3 0.35>
    componentLinearVelScale=<0.15 0.2 0.15>
    simSpaceSettings {
     maxAngularAcceleration=1000
     maxAngularVelocity=3
     maxLinearAcceleration=700
     maxLinearVelocity=500
     velocityScaleZ=5
     worldAlpha=1
     externalLinearDragV=<0.2 0.2>
    }
   }
   n12 = "Transform (Modify) Bone - Bone: jacket_fr_l_mstr"!(ComponentPose=@n11, bAlphaBoolEnabled:bool=Get bIsCrouching(), Translation="0,0,3", Rotation="0,0,0", Scale="1,1,1") -> (Pose) {
    alphaInputType=Bool
    translationMode=BMM_Additive
    translationSpace=BCS_BoneSpace
    alphaBoolBlend {blendOption=Linear}
    boneToModify {boneName=jacket_fr_l_mstr}
    translation {z=3}
   }
   n13 = "Constraint - belt_mstr (PPPPPPPP)"!(
    ComponentPose=@n12
    bAlphaBoolEnabled:bool=Get bIsCrouching()
    ConstraintWeights_0=0.0991
    ConstraintWeights_1=0.0263
    ConstraintWeights_2=0.6821
    ConstraintWeights_3=0
    ConstraintWeights_4=0.1905
    ConstraintWeights_5=0
    ConstraintWeights_6=0.0019
    ConstraintWeights_7=0
   ) -> (Pose) {
    alphaInputType=Bool
    alphaBoolBlend {blendOption=Linear}
    boneToModify {boneName=belt_mstr}
    constraintSetup [
     {targetBone {boneName=pelvis}}
     {targetBone {boneName=thigh_l}}
     {targetBone {boneName=thigh_twist_01_l}}
     {targetBone {boneName=deform_glute_l}}
     {targetBone {boneName=deform_groin_l}}
     {targetBone {boneName=thigh_r}}
     {targetBone {boneName=thigh_twist_01_r}}
     {targetBone {boneName=deform_groin_r}}
    ]
    constraintWeights [1,1,1,1,1,1,1,1]
   }
   n14 = "Constraint - belt_mstr (PPPPPPPP)"!(
    ComponentPose=@n13
    Alpha:float=0.5
    ConstraintWeights_0=0.0991
    ConstraintWeights_1=0.0263
    ConstraintWeights_2=0.6821
    ConstraintWeights_3=0
    ConstraintWeights_4=0.1905
    ConstraintWeights_5=0
    ConstraintWeights_6=0.0019
    ConstraintWeights_7=0
   ) -> (Pose) {
    alphaBoolBlend {blendOption=Linear}
    boneToModify {boneName=belt_mstr}
    constraintSetup [
     {targetBone {boneName=pelvis}}
     {targetBone {boneName=thigh_l}}
     {targetBone {boneName=thigh_twist_01_l}}
     {targetBone {boneName=deform_glute_l}}
     {targetBone {boneName=deform_groin_l}}
     {targetBone {boneName=thigh_r}}
     {targetBone {boneName=thigh_twist_01_r}}
     {targetBone {boneName=deform_groin_r}}
    ]
    constraintWeights [1,1,1,1,1,1,1,1]
   }
   n15 = Get skydiving new() -> (skydiving new:double)
   n16 = "Constraint - dyn_jacket_bk_01 (R)"(ComponentPose=@n14, Alpha=@n15, ConstraintWeights_0=1) {
    alphaBoolBlend {blendOption=Linear}
    boneToModify {boneName=dyn_jacket_bk_01}
    constraintSetup [{transformType=Rotation, targetBone {boneName=pelvis}}]
    constraintWeights [1]
   }
   n17 = "Constraint - dyn_jacket_bk_r_01 (R)"(ComponentPose=@n16, Alpha=@n15, ConstraintWeights_0=1) {
    alphaBoolBlend {blendOption=Linear}
    boneToModify {boneName=dyn_jacket_bk_r_01}
    constraintSetup [{transformType=Rotation, targetBone {boneName=pelvis}}]
    constraintWeights [1]
   }
   n18 = "Constraint - dyn_jacket_bk_l_01 (R)"(ComponentPose=@n17, Alpha=@n15, ConstraintWeights_0=1) {
    alphaBoolBlend {blendOption=Linear}
    boneToModify {boneName=dyn_jacket_bk_l_01}
    constraintSetup [{transformType=Rotation, targetBone {boneName=pelvis}}]
    constraintWeights [1]
   }
   n19 = Component To Local(ComponentPose=@n18)
   n20 = "Constraint - jacket_fr_l_mstr (R)"!(
    ComponentPose=Local To Component(LocalPose=Use cached pose 'Anim Dynamics'())
    bAlphaBoolEnabled:bool=Get bIsCrouching()
    ConstraintWeights_0=1
   ) -> (Pose) {
    alphaInputType=Bool
    alphaBoolBlend {blendOption=Linear}
    boneToModify {boneName=jacket_fr_l_mstr}
    constraintSetup [{transformType=Rotation, targetBone {boneName=thigh_twist_01_l}}]
    constraintWeights [1]
   }
   n21 = "Constraint - emblem_mstr (PPPPPPPP)"(
    ComponentPose=@n20
    Alpha=1
    ConstraintWeights_0=0.0828
    ConstraintWeights_1=0.1194
    ConstraintWeights_2=0.0635
    ConstraintWeights_3=0.1983
    ConstraintWeights_4=0.1315
    ConstraintWeights_5=0.1428
    ConstraintWeights_6=0.0281
    ConstraintWeights_7=0.2337
   ) {
    alphaBoolBlend {blendOption=Linear}
    boneToModify {boneName=emblem_mstr}
    constraintSetup [
     {targetBone {boneName=spine_03}}
     {targetBone {boneName=spine_04}}
     {targetBone {boneName=spine_05}}
     {targetBone {boneName=clavicle_l}}
     {targetBone {boneName=upperarm_twist_01_l}}
     {targetBone {boneName=deform_upperarm_fr_l}}
     {targetBone {boneName=upperarm_twist_02_l}}
     {targetBone {boneName=deform_pec_l}}
    ]
    constraintWeights [1,1,1,1,1,1,1,1]
   }
   n22 = "Constraint - dyn_emblem_02 (PPPPPPPP)"(
    ComponentPose=@n21
    Alpha=1
    ConstraintWeights_0=0.063
    ConstraintWeights_1=0.1685
    ConstraintWeights_2=0.2843
    ConstraintWeights_3=0.1618
    ConstraintWeights_4=0.118
    ConstraintWeights_5=0.0614
    ConstraintWeights_6=0.0255
    ConstraintWeights_7=0.1177
   ) {
    alphaBoolBlend {blendOption=Linear}
    boneToModify {boneName=dyn_emblem_02}
    constraintSetup [
     {targetBone {boneName=spine_01}}
     {targetBone {boneName=spine_02}}
     {targetBone {boneName=spine_03}}
     {targetBone {boneName=spine_04}}
     {targetBone {boneName=clavicle_l}}
     {targetBone {boneName=upperarm_twist_01_l}}
     {targetBone {boneName=upperarm_twist_02_l}}
     {targetBone {boneName=deform_pec_l}}
    ]
    constraintWeights [1,1,1,1,1,1,1,1]
   }
   n23 = Look At - Bone: dyn_emblem_01(ComponentPose=@n22, Alpha=1) {
    alphaBoolBlend {blendOption=Linear}
    boneToModify {boneName=dyn_emblem_01}
    lookAtTarget {boneReference {boneName=dyn_emblem_02}}
    lookAt_Axis {axis=<1 0>}
   }
   n24 = "Transform (Modify) Bone - Bone: neck_correction_mstr"(ComponentPose=@n23, Alpha=1, Translation="0,0,0.75", Rotation="0,0,0", Scale="1,1,1") {
    translationMode=BMM_Additive
    translationSpace=BCS_BoneSpace
    alphaBoolBlend {blendOption=Linear}
    boneToModify {boneName=neck_correction_mstr}
    translation {z=0.75}
   }
   n25 = Control Rig(
    Source=Component To Local(ComponentPose=@n24)
    SkyDiving_CTRL=Get bIsSkyDiving()
    Crouch_CTRL=Get bIsCrouching()
    RBAN Disabled=NOT Boolean(A=Get bRigidBodyEnabled())
    RBAN Enabled=Get bRigidBodyEnabled()
    Jumping_Falling=OR Boolean(A=Get bIsJumping(), B=Get bIsFalling())
    Accelerating=Get bIsAccelerating2D()
    Sliding=Get bIsSliding()
    Emoting=Get bIsPlayingEmote()
    Aiming=Get bIsTargeting()
   ) {
    alphaBoolBlend {blendOption=Linear}
    controlRigAssetReference {blueprintRigClass=&M_MED_BraveBuildSuper_Coat_ControlRig_C}
   }
   n26 = Get Bwd Movement() -> (Bwd Movement:double)
   n27 = Trail controller - Bone: dyn_jacket_bk_02!(ComponentPose=Local To Component(LocalPose=@n25), Alpha=@n26, FakeVelocity:Vector="0,0,0") -> (Pose) {
    bActorSpaceFakeVel=true
    bShowTrailLocation=true
    debugLifeTime=0.1
    maxDeltaTime=0.0333
    relaxationSpeedScale=3
    alphaScaleBiasClamp {bInterpResult=true}
    baseJoint {boneName=root}
    planarLimits [
     {drivingBone {boneName=root}, planeTransform {location=<0 0 10>, rotation=<rot 0 0 0>, scale=<1 1 1>}}
     {drivingBone {boneName=spine_03}, planeTransform {location=<0 -20 0>, rotation=<rot 0 15 -89.9999>, scale=<1 1 1>}}
    ]
    rotationLimits [{limitMax=<180 180 180>}, {limitMax=<180 180 180>}]
    rotationOffsets [<0 0 0>, <0 0 0>]
    trailBone {boneName=dyn_jacket_bk_02}
    trailRelaxationSpeed {editorCurveData {keys [{arriveTangent=-72.0739, interpMode=RCIM_Cubic, leaveTangent=-72.0739, tangentMode=RCTM_User, time=0, value=35}, <at 1 5 RCIM_Cubic>]}}
   }
   n28 = Trail controller - Bone: dyn_jacket_bk_l_02!(ComponentPose=@n27, Alpha=@n26, FakeVelocity:Vector="0,0,0") -> (Pose) {
    bActorSpaceFakeVel=true
    bShowTrailLocation=true
    debugLifeTime=0.1
    maxDeltaTime=0.0333
    relaxationSpeedScale=3
    alphaScaleBiasClamp {bInterpResult=true}
    baseJoint {boneName=root}
    planarLimits [
     {drivingBone {boneName=root}, planeTransform {location=<0 0 10>, rotation=<rot 0 0 0>, scale=<1 1 1>}}
     {drivingBone {boneName=spine_03}, planeTransform {location=<0 -20 0>, rotation=<rot 0 15 -89.9999>, scale=<1 1 1>}}
    ]
    rotationLimits [{limitMax=<180 180 180>}, {limitMax=<180 180 180>}]
    rotationOffsets [<0 0 0>, <0 0 0>]
    trailBone {boneName=dyn_jacket_bk_l_02}
    trailRelaxationSpeed {editorCurveData {keys [{arriveTangent=-72.0739, interpMode=RCIM_Cubic, leaveTangent=-72.0739, tangentMode=RCTM_User, time=0, value=35}, <at 1 5 RCIM_Cubic>]}}
   }
   n29 = Trail controller - Bone: dyn_jacket_bk_r_02!(ComponentPose=@n28, Alpha=@n26, FakeVelocity:Vector="0,0,0") -> (Pose) {
    bActorSpaceFakeVel=true
    bShowTrailLocation=true
    debugLifeTime=0.1
    maxDeltaTime=0.0333
    relaxationSpeedScale=3
    alphaScaleBiasClamp {bInterpResult=true}
    baseJoint {boneName=root}
    planarLimits [
     {drivingBone {boneName=root}, planeTransform {location=<0 0 10>, rotation=<rot 0 0 0>, scale=<1 1 1>}}
     {drivingBone {boneName=spine_03}, planeTransform {location=<0 -20 0>, rotation=<rot 0 15 -89.9999>, scale=<1 1 1>}}
    ]
    rotationLimits [{limitMax=<180 180 180>}, {limitMax=<180 180 180>}]
    rotationOffsets [<0 0 0>, <0 0 0>]
    trailBone {boneName=dyn_jacket_bk_r_02}
    trailRelaxationSpeed {editorCurveData {keys [{arriveTangent=-72.0739, interpMode=RCIM_Cubic, leaveTangent=-72.0739, tangentMode=RCTM_User, time=0, value=35}, <at 1 5 RCIM_Cubic>]}}
   }
   n30 = Get Right Movement() -> (Right Movement:double)
   n31 = Trail controller - Bone: dyn_jacket_bkout_r_02!(
    ComponentPose=@n29
    Alpha=FInterpEaseInOut!(A:double=@n26, B:double=@n30, Alpha:double=1, Exponent:double=0) -> (ReturnValue:double)
    FakeVelocity:Vector="0,0,0"
   ) -> (Pose) {
    bActorSpaceFakeVel=true
    bShowTrailLocation=true
    debugLifeTime=0.1
    maxDeltaTime=0.0333
    relaxationSpeedScale=3
    alphaScaleBiasClamp {bInterpResult=true}
    baseJoint {boneName=root}
    planarLimits [
     {drivingBone {boneName=root}, planeTransform {location=<0 0 10>, rotation=<rot 0 0 0>, scale=<1 1 1>}}
     {drivingBone {boneName=spine_03}, planeTransform {location=<0 -20 0>, rotation=<rot 0 15 -89.9999>, scale=<1 1 1>}}
    ]
    rotationLimits [{limitMax=<180 180 180>}, {limitMax=<180 180 180>}]
    rotationOffsets [<0 0 0>, <0 0 0>]
    trailBone {boneName=dyn_jacket_bkout_r_02}
    trailRelaxationSpeed {editorCurveData {keys [{arriveTangent=-72.0739, interpMode=RCIM_Cubic, leaveTangent=-72.0739, tangentMode=RCTM_User, time=0, value=35}, <at 1 5 RCIM_Cubic>]}}
   }
   n32 = Get Left Movement() -> (Left Movement:double)
   n33 = Trail controller - Bone: dyn_jacket_bkout_l_02!(ComponentPose=@n31, Alpha=FInterpEaseInOut(A=@n26, B=@n32, Alpha=1, Exponent=0), FakeVelocity:Vector="0,0,0") -> (Pose) {
    bActorSpaceFakeVel=true
    bShowTrailLocation=true
    debugLifeTime=0.1
    maxDeltaTime=0.0333
    relaxationSpeedScale=3
    alphaScaleBiasClamp {bInterpResult=true}
    baseJoint {boneName=root}
    planarLimits [
     {drivingBone {boneName=root}, planeTransform {location=<0 0 10>, rotation=<rot 0 0 0>, scale=<1 1 1>}}
     {drivingBone {boneName=spine_03}, planeTransform {location=<0 -20 0>, rotation=<rot 0 15 -89.9999>, scale=<1 1 1>}}
    ]
    rotationLimits [{limitMax=<180 180 180>}, {limitMax=<180 180 180>}]
    rotationOffsets [<0 0 0>, <0 0 0>]
    trailBone {boneName=dyn_jacket_bkout_l_02}
    trailRelaxationSpeed {editorCurveData {keys [{arriveTangent=-72.0739, interpMode=RCIM_Cubic, leaveTangent=-72.0739, tangentMode=RCTM_User, time=0, value=35}, <at 1 5 RCIM_Cubic>]}}
   }
   n34 = Trail controller - Bone: dyn_jacket_l_02!(ComponentPose=@n33, Alpha=@n32, FakeVelocity:Vector="0,0,0") -> (Pose) {
    bActorSpaceFakeVel=true
    bShowTrailLocation=true
    debugLifeTime=0.1
    maxDeltaTime=0.0333
    relaxationSpeedScale=3
    alphaScaleBiasClamp {bInterpResult=true}
    baseJoint {boneName=root}
    planarLimits [
     {drivingBone {boneName=root}, planeTransform {location=<0 0 10>, rotation=<rot 0 0 0>, scale=<1 1 1>}}
     {drivingBone {boneName=spine_03}, planeTransform {location=<0 -20 0>, rotation=<rot 0 15 -89.9999>, scale=<1 1 1>}}
    ]
    rotationLimits [{limitMax=<180 180 180>}, {limitMax=<180 180 180>}]
    rotationOffsets [<0 0 0>, <0 0 0>]
    trailBone {boneName=dyn_jacket_l_02}
    trailRelaxationSpeed {editorCurveData {keys [{arriveTangent=-72.0739, interpMode=RCIM_Cubic, leaveTangent=-72.0739, tangentMode=RCTM_User, time=0, value=35}, <at 1 5 RCIM_Cubic>]}}
   }
   n35 = Trail controller - Bone: dyn_jacket_fr_l_02!(ComponentPose=@n34, Alpha=@n32, FakeVelocity:Vector="0,0,0") -> (Pose) {
    bActorSpaceFakeVel=true
    bShowTrailLocation=true
    debugLifeTime=0.1
    maxDeltaTime=0.0333
    relaxationSpeedScale=3
    alphaScaleBiasClamp {bInterpResult=true}
    baseJoint {boneName=root}
    planarLimits [
     {drivingBone {boneName=root}, planeTransform {location=<0 0 10>, rotation=<rot 0 0 0>, scale=<1 1 1>}}
     {drivingBone {boneName=spine_03}, planeTransform {location=<0 -20 0>, rotation=<rot 0 15 -89.9999>, scale=<1 1 1>}}
    ]
    rotationLimits [{limitMax=<180 180 180>}, {limitMax=<180 180 180>}]
    rotationOffsets [<0 0 0>, <0 0 0>]
    trailBone {boneName=dyn_jacket_fr_l_02}
    trailRelaxationSpeed {editorCurveData {keys [{arriveTangent=-72.0739, interpMode=RCIM_Cubic, leaveTangent=-72.0739, tangentMode=RCTM_User, time=0, value=35}, <at 1 5 RCIM_Cubic>]}}
   }
   n36 = Trail controller - Bone: dyn_jacket_r_02!(ComponentPose=@n35, Alpha=@n30, FakeVelocity:Vector="0,0,0") -> (Pose) {
    bActorSpaceFakeVel=true
    bShowTrailLocation=true
    debugLifeTime=0.1
    maxDeltaTime=0.0333
    relaxationSpeedScale=3
    alphaScaleBiasClamp {bInterpResult=true}
    baseJoint {boneName=root}
    planarLimits [
     {drivingBone {boneName=root}, planeTransform {location=<0 0 10>, rotation=<rot 0 0 0>, scale=<1 1 1>}}
     {drivingBone {boneName=spine_03}, planeTransform {location=<0 -20 0>, rotation=<rot 0 15 -89.9999>, scale=<1 1 1>}}
    ]
    rotationLimits [{limitMax=<180 180 180>}, {limitMax=<180 180 180>}]
    rotationOffsets [<0 0 0>, <0 0 0>]
    trailBone {boneName=dyn_jacket_r_02}
    trailRelaxationSpeed {editorCurveData {keys [{arriveTangent=-72.0739, interpMode=RCIM_Cubic, leaveTangent=-72.0739, tangentMode=RCTM_User, time=0, value=35}, <at 1 5 RCIM_Cubic>]}}
   }
   n37 = Trail controller - Bone: dyn_jacket_fr_r_02!(ComponentPose=@n36, Alpha=@n30, FakeVelocity:Vector="0,0,0") -> (Pose) {
    bActorSpaceFakeVel=true
    bShowTrailLocation=true
    debugLifeTime=0.1
    maxDeltaTime=0.0333
    relaxationSpeedScale=3
    alphaScaleBiasClamp {bInterpResult=true}
    baseJoint {boneName=root}
    planarLimits [
     {drivingBone {boneName=root}, planeTransform {location=<0 0 10>, rotation=<rot 0 0 0>, scale=<1 1 1>}}
     {drivingBone {boneName=spine_03}, planeTransform {location=<0 -20 0>, rotation=<rot 0 15 -89.9999>, scale=<1 1 1>}}
    ]
    rotationLimits [{limitMax=<180 180 180>}, {limitMax=<180 180 180>}]
    rotationOffsets [<0 0 0>, <0 0 0>]
    trailBone {boneName=dyn_jacket_fr_r_02}
    trailRelaxationSpeed {editorCurveData {keys [{arriveTangent=-72.0739, interpMode=RCIM_Cubic, leaveTangent=-72.0739, tangentMode=RCTM_User, time=0, value=35}, <at 1 5 RCIM_Cubic>]}}
   }
   n38 = Component To Local(ComponentPose=@n37)
   n39 = Get bRigidBodyEnabled()
   n40 = "Transform (Modify) Bone - Bone: Inner_Jacket_Seem_Scale_Correction"(
    ComponentPose=Local To Component(LocalPose=Blend Poses by bool(BlendPose_0=@n19, BlendPose_1=@n38, bActiveValue=@n39) {blendType=Linear})
    Alpha=1
    Translation="0,0,0"
    Rotation="0,0,0"
    Scale="0,0,0"
   ) {
    scaleMode=BMM_Additive
    scaleSpace=BCS_BoneSpace
    translationMode=BMM_Additive
    translationSpace=BCS_BoneSpace
    alphaBoolBlend {blendOption=Linear}
    boneToModify {boneName=Inner_Jacket_Seem_Scale_Correction}
    scale=<0 0 0>
   }
   Output Pose(Result=Component To Local(ComponentPose=@n40)) {layerGroup=None}
   n41 = Control Rig!(Source=Copy Pose From Mesh() {bUseAttachedParent=true}, Alpha) -> (Pose) {
    alphaBoolBlend {blendOption=Linear}
    controlRigAssetReference {blueprintRigClass=/BRCosmetics/Characters/Player/Male/Medium/Bodies/M_MED_BraveBuild/Meshes/M_MED_BraveBuild_Deform_ControlRig.M_MED_BraveBuild_Deform_ControlRig_C}
   }
   n42 = "Constraint - jacket_zipper_mstr (PPPPP)"!(
    ComponentPose=Local To Component(LocalPose=@n41)
    Alpha=1
    ConstraintWeights_0=0.4112
    ConstraintWeights_1:float=0.2961
    ConstraintWeights_2:float=0.1159
    ConstraintWeights_3:float=0.1146
    ConstraintWeights_4:float=0.0622
   ) -> (Pose) {
    alphaBoolBlend {blendOption=Linear}
    boneToModify {boneName=jacket_zipper_mstr}
    constraintSetup [
     {targetBone {boneName=spine_03}}
     {targetBone {boneName=spine_02}}
     {targetBone {boneName=spine_01}}
     {targetBone {boneName=spine_04}}
     {targetBone {boneName=pelvis}}
    ]
    constraintWeights [1,1,1,1,1]
   }
   n43 = Anim Dynamics - Bone: jacket_zipper_mstr(ComponentPose=@n42, Alpha=1, ExternalForce="0,0,0") {
    angularDampingOverride=0.9999
    bOverrideAngularDamping=true
    bUseGravityOverride=true
    bUsePlanarLimit=false
    simulationSpace=RootRelative
    alphaBoolBlend {blendOption=Linear}
    boundBone {boneName=jacket_zipper_mstr}
    componentAppliedLinearAccClamp=<3000 3000 3000>
    componentLinearAccScale=<0.5 0.5 0.5>
    componentLinearVelScale=<0.25 0.25 0.25>
    gravityOverride {z=-980}
    physicsBodyDefinitions [{
      boundBone {boneName=jacket_zipper_mstr}
      boxExtents=<2 10 10>
      constraintSetup {angularLimitsMax=<2 5 30>, angularLimitsMin=<-2 -30 -30>, angularTarget=<0 0 0>}
      localJointOffset=<3 0 0>
     }]
   }
   n44 = Anim Dynamics - Bone: arm_l_lower_sleeve_dyn(ComponentPose=@n43, Alpha=0.5, ExternalForce="0,0,0") {
    angularDampingOverride=0.998
    bGravityOverrideInSimSpace=true
    bOverrideAngularDamping=true
    bUseGravityOverride=true
    simulationSpace=RootRelative
    alphaBoolBlend {blendOption=Linear}
    boundBone {boneName=arm_l_lower_sleeve_dyn}
    componentAppliedLinearAccClamp=<3000 3000 3000>
    componentLinearAccScale=<0.3 0.3 0.3>
    componentLinearVelScale=<0.3 0.3 0.3>
    gravityOverride {z=-980}
    physicsBodyDefinitions [{
      sphereCollisionRadius=0
      boundBone {boneName=arm_l_lower_sleeve_dyn}
      boxExtents=<12 6 12>
      constraintSetup {angularLimitsMax=<0 5 5>, angularLimitsMin=<0 -5 -5>, angularTarget=<0 0 0>, linearAxesMax=<0 1 1>, linearAxesMin=<0 -1 -1>}
      localJointOffset=<1 0 0>
     }]
    relativeSpaceBone {boneName=lowerarm_r}
   }
   n45 = Anim Dynamics - Bone: arm_r_lower_sleeve_mstr(ComponentPose=@n44, Alpha=0.5, ExternalForce="0,0,0") {
    angularDampingOverride=0.998
    bGravityOverrideInSimSpace=true
    bOverrideAngularDamping=true
    bUseGravityOverride=true
    simulationSpace=RootRelative
    alphaBoolBlend {blendOption=Linear}
    boundBone {boneName=arm_r_lower_sleeve_mstr}
    componentAppliedLinearAccClamp=<3000 3000 3000>
    componentLinearAccScale=<0.3 0.3 0.3>
    componentLinearVelScale=<0.3 0.3 0.3>
    gravityOverride {z=-980}
    physicsBodyDefinitions [{
      sphereCollisionRadius=0
      boundBone {boneName=arm_r_lower_sleeve_mstr}
      boxExtents=<18 9 18>
      constraintSetup {angularLimitsMax=<0 5 5>, angularLimitsMin=<0 -5 -5>, angularTarget=<0 0 0>, linearAxesMax=<0 1 1>, linearAxesMin=<0 -1 -1>}
      localJointOffset=<1 0 0>
     }]
    relativeSpaceBone {boneName=lowerarm_r}
   }
   n46 = Anim Dynamics - Bone: leg_l_lower_sleeve_mstr(ComponentPose=@n45, Alpha=0.5, ExternalForce="0,0,0") {
    angularDampingOverride=0.998
    bGravityOverrideInSimSpace=true
    bOverrideAngularDamping=true
    bUseGravityOverride=true
    simulationSpace=RootRelative
    alphaBoolBlend {blendOption=Linear}
    boundBone {boneName=leg_l_lower_sleeve_mstr}
    componentAppliedLinearAccClamp=<3000 3000 3000>
    componentLinearAccScale=<0.3 0.3 0.3>
    componentLinearVelScale=<0.3 0.3 0.3>
    gravityOverride {z=-980}
    physicsBodyDefinitions [{
      sphereCollisionRadius=0
      boundBone {boneName=leg_l_lower_sleeve_mstr}
      boxExtents=<18 9 18>
      constraintSetup {angularLimitsMax=<0 5 5>, angularLimitsMin=<0 -5 -5>, angularTarget=<0 0 0>, linearAxesMax=<0 1 1>, linearAxesMin=<0 -1 -1>}
      localJointOffset=<1 0 0>
     }]
    relativeSpaceBone {boneName=lowerarm_r}
   }
   n47 = Anim Dynamics - Bone: leg_r_lower_sleeve_mstr(ComponentPose=@n46, Alpha=0.5, ExternalForce="0,0,0") {
    angularDampingOverride=0.998
    bGravityOverrideInSimSpace=true
    bOverrideAngularDamping=true
    bUseGravityOverride=true
    simulationSpace=RootRelative
    alphaBoolBlend {blendOption=Linear}
    boundBone {boneName=leg_r_lower_sleeve_mstr}
    componentAppliedLinearAccClamp=<3000 3000 3000>
    componentLinearAccScale=<0.3 0.3 0.3>
    componentLinearVelScale=<0.3 0.3 0.3>
    gravityOverride {z=-980}
    physicsBodyDefinitions [{
      sphereCollisionRadius=0
      boundBone {boneName=leg_r_lower_sleeve_mstr}
      boxExtents=<12 6 12>
      constraintSetup {angularLimitsMax=<0 5 5>, angularLimitsMin=<0 -5 -5>, angularTarget=<0 0 0>, linearAxesMax=<0 1 1>, linearAxesMin=<0 -1 -1>}
      localJointOffset=<1 0 0>
     }]
    relativeSpaceBone {boneName=lowerarm_r}
   }
   Save cached pose 'Anim Dynamics'(Pose=Component To Local(ComponentPose=@n47)) {cacheName=Anim Dynamics}
  ]}
 EventGraph {nodes [
   n1 = Event BlueprintUpdateAnimation()
   > n2 = Sequence!(execute) -> (then_0, then_1, then_2:Exec, then_3:Exec, then_4:Exec)
   > n3 = Gravity Override State Function DEPRECATED(CharacterPartAnimInstance=Self-Reference(), GameTimeInSeconds=GetGameTimeInSeconds(), GravityOverrideParams=Get Gravity Override Params())
   > Set Gravity Override(Gravity Override=@n3)
   Is Valid(exec=@n2.then_1, InputObject=Get Player Pawn() -> (Player Pawn:Fort Player Pawn Athena Object Reference))
   > Set WPO Globals Function(
    Material=Get Mat Instance Dynamic()
    Overlay Material=Get Overlay Mat Instance Dynamic()
    Skeletal Mesh=Get Skel Mesh()
    "Local Noise Transform  Location"=Get Local Noise Transform Location() -> (Local Noise Transform Location:Vector)
    Local Noise Offset=Get Local Noise Offset()
   )
   > Set WPO Regional Function!(
    execute
    Material:Material Instance Dynamic Object Reference=Get Mat Instance Dynamic()
    Overlay Material:Material Instance Dynamic Object Reference=Get Overlay Mat Instance Dynamic()
    Skeletal Mesh:Skeletal Mesh Component Object Reference=Get Skel Mesh()
    Velocity Transform:Name=foot_r
    Velocity Interp Speed:double=Get Velocity Interp Speed()
    Collision Transform:Name=calf_r
    Collision Transform Axis:EAxis Enum
    Invert Collision Transform Axis:bool=false
    Delta Time X:double=@n1.DeltaTimeX
    Previous Velocity Transform Location:"Vector (by ref)"=Get R1 Velocity Transform Location()
    Previous Velocity Transform Velocity:"Vector (by ref)"=Get R1 Velocity Transform Velocity()
    Region:int=1
   ) -> (then)
   > Set WPO Regional Function(
    Material=Get Mat Instance Dynamic()
    Overlay Material=Get Overlay Mat Instance Dynamic()
    Skeletal Mesh=Get Skel Mesh()
    Velocity Transform=hand_l
    Velocity Interp Speed=Get Velocity Interp Speed()
    Collision Transform=lowerarm_l
    Invert Collision Transform Axis=true
    Delta Time X=@n1.DeltaTimeX
    Previous Velocity Transform Location=Get R2 Velocity Transform Location()
    Previous Velocity Transform Velocity=Get R2 Velocity Transform Velocity()
    Region=2
   )
   > n4 = Branch(Condition=Get bIsSkydiveFloating())
   > n5 = Set WPO Pho Wind Intensity(WPO Pho Wind Intensity=1)
   SetScalarParameterValue(execute=[@n6.then, @n5.then], self=Get Mat Instance Dynamic(), ParameterName=ClothSim_WindStrength, Value=Get WPO Pho Wind Intensity())
   n6 = Set WPO Pho Wind Intensity(execute=@n4.else, WPO Pho Wind Intensity=0)
   n7 = Branch(execute=@n2.then_2, Condition=Get bIsSkyDiving())
   > Set skydiving new(skydiving new:double=0.3) -> (Output_Get:double)
   Set skydiving new(execute=@n7.else, skydiving new=0)
   n8 = Branch(execute=@n2.then_3, Condition=Get bIsSkyDiving())
   > n9 = Branch(Condition="float > float"(A=Get PawnVelocityZ(), B=50) {is=title K2Node_CallFunction})
   > Set Bwd Movement(Bwd Movement:double=float * float(A=Get Bwd Movement(), B=0)) -> (Output_Get:double)
   Set Bwd Movement(execute=[@n8.else, @n9.else], Bwd Movement=float * float(A=Get Bwd Movement(), B=1))
   n10 = PlayerMovement(execute=@n2.then_4, AnimInstance=Self-Reference(), XDotProductCoefficient=1, YDotProductCoefficient=1, ZDotProductCoefficient=1)
   n11 = MapRangeClamped(Value=@n10.YMovement, InRangeA=0, InRangeB=0)
   n12 = float * float(A="Absolute (Float)"(A=@n10), B=0.5)
   > Set Bwd Movement(Bwd Movement=MapRangeClamped(
     Value=float + float(A=@n12, B=@n11) {is=title K2Node_CommutativeAssociativeBinaryOperator}
     InRangeA=0
     InRangeB=1.1
     OutRangeA=0.1
     OutRangeB=0
    ))
   n13 = MapRangeClamped(Value=@n10, InRangeA=-0.2, InRangeB=1)
   n14 = float * float(A="Absolute (Float)"(A=@n10.YMovement), B=0.2)
   > Set Right Movement(Right Movement:double=MapRangeClamped(Value=float + float(A=@n14, B=@n13), InRangeA=0, InRangeB=1.2, OutRangeB=0)) -> (Output_Get:double)
   n15 = MapRangeClamped(Value=@n10, InRangeA=0, InRangeB=0, OutRangeB=0.3)
   n16 = float * float(A="Absolute (Float)"(A=@n10.YMovement), B=0.2)
   > Set Left Movement(Left Movement:double=MapRangeClamped(Value=float + float(A=@n16, B=@n15), InRangeA=0.2, InRangeB=1.2, OutRangeB=0.5)) -> (Output_Get:double)
   Event BlueprintInitializeAnimation()
   > Set Skel Mesh(Skel Mesh=GetOwnerSkeletalMeshComponent())
   n17 = TryGetPawnOwner()
   > Is Valid(InputObject=@n17)
   > n18 = Cast To FortPlayerPawnAthena(Object=@n17)
   > Set Player Pawn(Player Pawn:Fort Player Pawn Athena Object Reference=@n18) -> (Output_Get:Fort Player Pawn Athena Object Reference)
   Event OnCharacterCustomizationLoaded()
   > n19 = Set Skel Mesh(Skel Mesh=GetPartSkeletalMeshComponent())
   > n20 = Get Dynamic Material Instance(SkeletalMeshComponent=@n19)
   > Set Mat Instance Dynamic(Mat Instance Dynamic=@n20)
   > n21 = GetOrCreateDynamicOverlayMaterial(MeshComponent=@n19, MaterialName=None)
   > Set Overlay Mat Instance Dynamic(Overlay Mat Instance Dynamic=@n21)
   MapRangeClamped(Value=0, InRangeA=0, InRangeB=1.2, OutRangeB=0.5)
   Branch(Condition=Get bIsInFrontEnd())
  ]}
]
variables [
 Bwd Movement {type=number}
 Gravity Override {type=Vector}
 Gravity Override Params {type=GravityOverrideParamsStruct, default {onGround_Moving {gravityOverride=<0 0 -1300>}}}
 Left Movement {type=number}
 Local Noise Offset {type=Vector}
 Local Noise Transform Location {type=Vector}
 Mat Instance Dynamic {type=MaterialInstanceDynamic}
 Overlay Mat Instance Dynamic {type=MaterialInstanceDynamic}
 Player Pawn {type=FortPlayerPawnAthena}
 R1 Velocity Transform Location {type=Vector}
 R1 Velocity Transform Velocity {type=Vector}
 R2 Velocity Transform Location {type=Vector}
 R2 Velocity Transform Velocity {type=Vector}
 Right Movement {type=number}
 Skel Mesh {type=SkeletalMeshComponent}
 skydiving new {type=number}
 Velocity Interp Speed {type=number}
 WPO Pho Wind Intensity {type=number, default=1}
]
